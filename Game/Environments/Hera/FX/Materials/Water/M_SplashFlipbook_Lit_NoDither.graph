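# Masked, lit water splash flipbook material for Niagara sprites, with responsive AA enabled.
# Particle color tints the flipbook's powered green channel, while green, alpha and particle alpha form an erodable mask.
# Dynamic parameters drive erosion, normal flattening, emissive and a particle-direction motion-blur trick using PreviousFrameSwitch on normals.
# Despite the NoDither name, squared near-camera fade, optionally scaled by DitherFade, feeds a Bayer dither below a threshold.
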Material /Game/Environments/Hera/FX/Materials/Water/M_SplashFlipbook_Lit_NoDither
graph {
 properties {
  bEnableResponsiveAA=true
  bUsedWithNiagaraSprites=true
  bUsedWithStaticMesh=false
  blendMode=BLEND_Masked
  previewMesh=/Game/Environments/Hera/FX/ZeroPoint/WaterDisturbance/SM_WaveMesh_Cut
  usedShadingModels=Default Lit
 }
 nodes [
  LinearInterpolate(B=ScalarParameter() {defaultValue=0.1, parameterName=EmissiveColor}, Alpha=ScalarParameter() {parameterName=UseEmissive})
  n1 = VertexNormalWS()
  n2 = Multiply(
   A=ParticleDirection()
   B=DynamicParameter() {defaultValue=<rgba 1 1 1 0>, paramNames [MaskErode, FlattenNormal, MotionBlur_Strength, Emissive]}.MotionBlur_Strength
  )
  LinearInterpolate(
   A=PreviousFrameSwitch(Current Frame=VertexNormalWS(), Previous Frame=Subtract(A=@n1, B=@n2))
   B=Constant()
   Alpha=ScalarParameter() {parameterName=DisableMotionBlur}
  )
  MaterialFunctionCall() {materialFunction=&FlattenNormal}
  Reroute(None=MaterialFunctionCall(
    Normal=TextureSample() {samplerType=SAMPLERTYPE_Normal, texture=/Game/Effects/Fort_Effects/Textures/Water/T_waterSplash_R_N}.RGB
    Flatness=DynamicParameter() {defaultValue=<rgba 1 1 1 0>, paramNames [MaskErode, FlattenNormal, MotionBlur_Strength, Emissive]}.FlattenNormal
   ) {materialFunction=&FlattenNormal}.Result)
  n3 = ParticleColor()
  n4 = TextureSampleParameter2D() {parameterName=Param}
  n5 = ScalarParameter() {defaultValue=0.2, parameterName=FlipBook_Power}
  Reroute(None=Multiply(A=@n3.RGB, B=Power(Base=@n4.G, Exp=@n5)))
  Reroute(None=Multiply(A=@n3.RGB, B=DynamicParameter() {defaultValue=<rgba 1 1 1 0>, paramNames [MaskErode, FlattenNormal, MotionBlur_Strength, Emissive]}.Emissive))
  n6 = DynamicParameter() {defaultValue=<rgba 1 1 1 0>, paramNames [MaskErode, FlattenNormal, MotionBlur_Strength, Emissive]}
  n7 = Multiply(A=@n3.A, B=Multiply(A=Power(Base=@n4.G, Exp=@n6.MaskErode), B=@n4.A))
  n8 = ScalarParameter() {defaultValue=0.5, parameterName=MaskScale}
  n9 = MaterialFunctionCall(Distance=ScalarParameter() {parameterName=CameraFadeDistance}) {materialFunction=&Near_Camera_Fade}
  n10 = Multiply(A=@n9.Result, B=@n9.Result)
  n11 = LinearInterpolate(
   A=@n10
   B=Multiply(A=@n10, B=DynamicParameter() {parameterIndex=1, defaultValue=<rgba 1 1 1 0>, paramNames [DitherFade,Param2,Param3,Param4]}.DitherFade)
   Alpha=ScalarParameter() {parameterName=EnableDitherFade}
  )
  n12 = Constant() {r=1}
  n13 = MaterialFunctionCall!(In=@n11, Iterations, Scale) {materialFunction=/Game/Packages/Fortress_SharedMaterials/Base_Material_Functions/MF_DitherBayerNoTex}
  Reroute(None=Multiply(A=Multiply(A=@n7, B=@n8), B=If(A=@n11, "A > B"=@n12, "A == B"=@n13.Result, "A < B"=@n13.Result) {constB=0.9}))
  ScalarParameter() {defaultValue=1, parameterName=Specular}
  ScalarParameter() {parameterName=Roughness}
  ScalarParameter() {parameterName=Metallic}
 ]
}
