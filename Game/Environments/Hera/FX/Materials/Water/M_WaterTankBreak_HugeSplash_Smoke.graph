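# Masked, lit Niagara sprite material for the smoke-like mist of a huge water tank break splash.
# A falling-mist 8x8 texture's green channel is shaped by Refine controls and multiplied by particle alpha, then fed into a Bayer dither to mimic soft opacity.
# Dynamic parameters drive refinement, depth fade, emissive and motion blur strength.
# The graph also holds particle-color tinting, a normal map, shading-path-dependent depth fade, and a previous-frame offset along particle direction.
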Material /Game/Environments/Hera/FX/Materials/Water/M_WaterTankBreak_HugeSplash_Smoke
graph {
 properties {
  bUseTranslucencyVertexFog=false
  bUsedWithNiagaraSprites=true
  bUsedWithStaticMesh=false
  blendMode=BLEND_Masked
  pixelDepthOffsetMode=PDOM_Legacy
  usedShadingModels=Default Lit
 }
 nodes [
  n1 = ParticleColor()
  MaterialFunctionCall(Emissive=@n1.RGB) {materialFunction=&MF_FX_EmissiveAdjustments}
  n2 = TextureCoordinate()
  n3 = TextureSample(UVs=@n2) {samplerType=SAMPLERTYPE_LinearColor, texture=/Game/Effects/Fort_Effects/Textures/Smoke/T_FallingMist_8x8_01}
  n4 = VectorParameter() {parameterName=Refine, defaultValue=<rgba 1.5 1 1 0.5>}
  n5 = LinearInterpolate(
   A=Multiply(A=Power(Base=@n3.G, Exp=@n4.R), B=@n4.G)
   B=Multiply(B=@n4.B)
   Alpha=DynamicParameter() {defaultValue=<rgba 0.5 10 1 0>, paramNames [Refine, DepthFade, MotionBlur_Strength, Emissive]}.Refine
  )
  n6 = Multiply(A=@n1.A, B=@n5)
  MaterialFunctionCall!(
   In=@n6
   Iterations=ScalarParameter() {defaultValue=4, parameterName="Iterations (S)"}
   Scale=ScalarParameter() {defaultValue=4, parameterName="Iterations (S)"}
  ) {materialFunction=/Game/Packages/Fortress_SharedMaterials/Base_Material_Functions/MF_DitherBayerNoTex}
  Multiply(A=@n1.RGB, B=DynamicParameter() {defaultValue=<rgba 0.5 10 1 0>, paramNames [Refine, DepthFade, MotionBlur_Strength, Emissive]}.Emissive)
  n7 = ScalarParameter() {parameterName=Add}
  Multiply(A=@n1.RGB, B=Multiply(A=@n3.G, B=Add(A=@n5, B=@n7)))
  n8 = ParticleDirection()
  n9 = DynamicParameter() {defaultValue=<rgba 0.5 10 1 0>, paramNames [Refine, DepthFade, MotionBlur_Strength, Emissive]}
  PreviousFrameSwitch(Current Frame=VertexNormalWS(), Previous Frame=Subtract(A=VertexNormalWS(), B=Multiply(A=@n8, B=@n9.MotionBlur_Strength)))
  ScalarParameter() {defaultValue=10, parameterName=FadeDistance}
  n10 = DepthFade(
   Opacity=@n6
   FadeDistance=DynamicParameter() {defaultValue=<rgba 0.5 10 1 0>, paramNames [Refine, DepthFade, MotionBlur_Strength, Emissive]}.DepthFade
  )
  ShadingPathSwitch(Default=@n10, Deferred=@n10, Forward=@n6, Mobile=@n6)
  Subtract(A=@n2, B=Multiply() {constA=1})
  TextureSample() {samplerType=SAMPLERTYPE_Normal, texture=/Game/Effects/Fort_Effects/Textures/Smoke/T_FallingMist_8x8_01_N}
 ]
}
