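# Translucent, two-sided lit water material, named for the waterfall part of a huge water-tank-break splash and usable on Niagara mesh particles.
# Two summed panning noises and a separate top noise distort V-based edge masks, whose product feeds dithering alongside particle alpha.
# Vertex-normal displacement from a panned texture blends start and end settings, with a powered-V gradient declared as StartOrEnd.
# It also builds flattened wave normals and Fresnel-blended emissive color with a ray-traced Lumen boost.
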Material /Game/Environments/Hera/FX/Materials/Water/M_WaterTankBreak_HugeSplash_Waterfall_Transprent
graph {
 properties {
  bUsedWithInstancedStaticMeshes=true
  bUsedWithNiagaraMeshParticles=true
  bUsedWithStaticLighting=true
  blendMode=BLEND_Translucent
  pixelDepthOffsetMode=PDOM_Legacy
  refractionMethod=RM_2DOffset
  translucencyLightingMode=TLM_SurfacePerPixelLighting
  twoSided=true
  usedShadingModels=Default Lit
 }
 nodes [
  Constant3Vector()
  Fresnel(ExponentIn=ScalarParameter() {defaultValue=3, parameterName=SpecularFresnel})
  n1 = MaterialFunctionCall(Float2=TextureCoordinate()) {materialFunction=&BreakOutFloat2Components}
  n2 = TextureObjectParameter() {
   group=EdgeMaskNoise
   parameterName=EdgeMaskNoise_Texture
   sortPriority=0
   texture=/Game/Global/Textures/Default/Tileable_Patterns/T_Smoke_Puff_Tile_Up
  }
  n3 = VectorParameter() {group=EdgeMaskNoise, parameterName=EdgeMaskNoise_UVControl, channelNames {a=VPan, b=UPan, g=VScale, r=UScale}, defaultValue=<rgba 1 1 0 0>}
  n4 = MaterialFunctionCall(Texture=@n2, UVControl_RGB=@n3, UVControl_A=@n3) {materialFunction=&FXCommon_TextureSample}
  n5 = VectorParameter() {
   group=EdgeMaskNoise
   parameterName=EdgeMaskNoise_NoiseControl
   sortPriority=60
   channelNames {a=VStrength, b=UStrength, g=VOffset, r=UOffset}
   defaultValue=<rgba -0.5 -0.5 0.01 0.01>
  }
  n6 = TextureObjectParameter() {
   group=EdgeMaskNoiseB
   parameterName=EdgeMaskNoiseB_Texture
   sortPriority=0
   texture=&T_Smoke_Puff_Tile_Up
  }
  n7 = VectorParameter() {group=EdgeMaskNoiseB, parameterName=EdgeMaskNoiseB_UVControl, channelNames {a=VPan, b=UPan, g=VScale, r=UScale}, defaultValue=<rgba 1 1 0 0>}
  n8 = MaterialFunctionCall(Texture=@n6, UVControl_RGB=@n7, UVControl_A=@n7) {materialFunction=&FXCommon_TextureSample}
  n9 = VectorParameter() {
   group=EdgeMaskNoiseB
   parameterName=EdgeMaskNoiseB_NoiseControl
   sortPriority=60
   channelNames {a=VStrength, b=UStrength, g=VOffset, r=UOffset}
   defaultValue=<rgba -0.5 -0.5 0.01 0.01>
  }
  n10 = Add(
   A=MaterialFunctionCall!(NoiseValue=AppendVector(A=@n4.R, B=@n4.R), Control_RGB=@n5, Control_A=@n5) {materialFunction=/Game/Effects/Niagara/MasterMaterials/Functions/Common/FXCommon_NoiseControl}.NoiseValue
   B=MaterialFunctionCall(NoiseValue=AppendVector(A=@n8.R, B=@n8.R), Control_RGB=@n9, Control_A=@n9) {materialFunction=&FXCommon_NoiseControl}.NoiseValue
  )
  n11 = VectorParameter() {
   group=EdgeMask
   parameterName=EdgeMask_EdgeMask_InvertControl
   channelNames {a="", b=Power, g=Multiply, r=Offset}
   defaultValue=<rgba 0 3 1 1>
  }
  n12 = VectorParameter() {group=EdgeMask, parameterName=EdgeMask_EdgeMask_Control, channelNames {a="", b=Power, g=Multiply, r=Offset}, defaultValue=<rgba 0 9 1 1>}
  n13 = VectorParameter() {
   group=EdgeMaskTopNoise
   parameterName=EdgeMaskTopNoise_UVControl
   channelNames {a=VPan, b=UPan, g=VScale, r=UScale}
   defaultValue=<rgba 1 1 0 0>
  }
  n14 = MaterialFunctionCall(
   Texture=TextureObjectParameter() {group=EdgeMaskTopNoise, parameterName=EdgeMaskTopNoise_Texture, sortPriority=0}
   UVControl_RGB=@n13
   UVControl_A=@n13
  ) {materialFunction=&FXCommon_TextureSample}
  n15 = VectorParameter() {
   group=EdgeMaskTopNoise
   parameterName=EdgeMaskTopNoise_NoiseControl
   sortPriority=60
   channelNames {a=VStrength, b=UStrength, g=VOffset, r=UOffset}
   defaultValue=<rgba -0.5 -0.5 0.01 0.01>
  }
  n16 = MaterialFunctionCall(NoiseValue=AppendVector(A=@n14.R, B=@n14.R), Control_RGB=@n15, Control_A=@n15) {materialFunction=&FXCommon_NoiseControl}
  n17 = VectorParameter() {
   group=EdgeMaskTop
   parameterName=EdgeMaskTop_EdgeMask_InvertControl
   channelNames {a="", b=Power, g=Multiply, r=Offset}
   defaultValue=<rgba 0 1 1 1>
  }
  n18 = VectorParameter() {
   group=EdgeMaskTop
   parameterName=EdgeMaskTop_EdgeMask_Control
   channelNames {a="", b=Power, g=Multiply, r=Offset}
   defaultValue=<rgba 0 1 1 1>
  }
  n19 = Multiply(
   A=MaterialFunctionCall!(UorV=Add(A=@n1.G, B=@n10) {constB=0}, Invert_EdgeControl=@n11, NoInvert_EdgeControl=@n12) {materialFunction=/Game/Effects/Niagara/MasterMaterials/Functions/Common/FXCommon_EdgeMask_Ex}.Mask
   B=MaterialFunctionCall(UorV=Add(A=@n1.G, B=@n16.NoiseValue), Invert_EdgeControl=@n17, NoInvert_EdgeControl=@n18) {materialFunction=&FXCommon_EdgeMask_Ex}.Mask
  )
  n20 = MaterialFunctionCall!(Alpha Threshold=@n19, Random) {materialFunction=/Game/TimeOfDay/CloudsAndStorm/Materials/MaterialFunctions/DitherTemporalAABlue}
  n21 = Constant()
  If("A > B"=If(A=@n19, B=Constant() {r=0.99}, "A > B"=Constant() {r=1}, "A == B"=@n20.Result, "A < B"=@n20.Result), "A == B"=@n21, "A < B"=@n21) {constB=0.1}
  n22 = TextureObjectParameter() {
   group=MainWave
   parameterName=MainWave_Texture
   samplerType=SAMPLERTYPE_LinearColor
   sortPriority=0
   texture=/Game/Global/Textures/Default/Tileable_Patterns/T_Water_TilingNormal_With_Height_02_Softened
  }
  n23 = VectorParameter() {group=MainWave, parameterName=MainWave_UVControl, channelNames {a=VPan, b=UPan, g=VScale, r=UScale}, defaultValue=<rgba 1 1 0 0>}
  MaterialFunctionCall(
   Normal=MaterialFunctionCall(Texture=@n22, UVControl_RGB=@n23, UVControl_A=@n23) {materialFunction=&FXCommon_TextureSample}.RGB
   Flatness=ScalarParameter() {defaultValue=1, parameterName=FlattenNormal}
  ) {materialFunction=&FlattenNormal}
  MaterialFunctionCall() {materialFunction=&MF_FX_EmissiveAdjustments}
  n24 = LinearInterpolate(A=ScalarParameter() {defaultValue=0.9}, Alpha=@n19)
  MaterialFunctionCall!(
   In=Multiply(
    A=ParticleColor().A
    B=MaterialFunctionCall(In=@n24) {materialFunction=/Game/Packages/Fortress_SharedMaterials/Base_Material_Functions/MF_DitherBayerNoTex}.Result
   )
   Iterations
   Scale
  ) {materialFunction=&MF_DitherBayerNoTex}
  n25 = TextureCoordinate()
  n26 = VectorParameter() {group=WPO, parameterName=WPO_UVControl, channelNames {a=VPan, b=UPan, g=VScale, r=UScale}, defaultValue=<rgba 1 1 0 0>}
  n27 = VectorParameter() {group=WPO, parameterName=WPO_UVControl_End, channelNames {a=VPan, b=UPan, g=VScale, r=UScale}, defaultValue=<rgba 1 1 0 0>}
  n28 = NamedRerouteUsage()
  n29 = ComponentMask(None=LinearInterpolate(A=@n26, B=@n27, Alpha=@n28)) {g=true, r=true}
  n30 = AppendVector(A=@n26.UPan, B=@n26.UPan)
  n31 = AppendVector(A=@n27.UPan, B=@n27.UPan)
  n32 = TextureSampleParameter2D(UVs=Panner(Coordinate=Multiply(A=@n25, B=@n29), Speed=LinearInterpolate(A=@n30, B=@n31, Alpha=@n28))) {parameterName=WPO_Texture, samplerType=SAMPLERTYPE_Normal, texture=/Game/Global/Textures/Default/Tileable_Patterns/1Normal}
  n33 = LinearInterpolate(
   A=ScalarParameter() {group=WPO, parameterName=WPO_GrayAdjust_Add}
   B=ScalarParameter() {group=WPO, parameterName=WPO_GrayAdjust_Add_End}
   Alpha=NamedRerouteUsage()
  )
  n34 = ScalarParameter() {defaultValue=1, group=WPO, parameterName=WPO_GrayAdjust_Mul}
  n35 = ScalarParameter() {defaultValue=1, group=WPO, parameterName=WPO_GrayAdjust_Mul_End}
  n36 = NamedRerouteUsage()
  Multiply(A=VertexNormalWS(), B=Multiply(A=Add(A=@n32.RGB, B=@n33), B=LinearInterpolate(A=@n34, B=@n35, Alpha=@n36)))
  Multiply(A=@n19)
  n37 = TextureCoordinate()
  n38 = ScalarParameter() {defaultValue=24, group=WPO, parameterName=WPO_BiasStartOrEnd}
  NamedRerouteDeclaration(None=ComponentMask(None=Power(Base=@n37, Exp=@n38)) {g=true}) {name=StartOrEnd, nodeColor=<rgba 0 1 0.3906 1>}
  n39 = Multiply(B=LinearInterpolate(
    A=VectorParameter() {parameterName=EmmisiveMiddle, defaultValue=<rgba 0 15 0 1>}.RGB
    B=VectorParameter() {parameterName=Emmisive, defaultValue=<rgba 0 15 0 1>}.RGB
    Alpha=Fresnel() {exponent=2}
   ))
  RayTracingQualitySwitch(Normal=@n39, RayTraced=Multiply(A=@n39, B=ScalarParameter() {defaultValue=3, parameterName=LumenBoost}))
  ScalarParameter() {parameterName=Roughness}
 ]
}
